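# Physics asset for the Little Egg medium female player body, covering dangling accessories: bracelets, cuffs, necklace, zipper, belt and anklet.
# Each accessory pairs a small capsule on a dynamic bone with one on its master bone, mostly constrained by tight swing and twist limits.
# The right bracelet forms a longer chain, with its master free to rotate against the main bracelet, which is angularly locked to its own master.
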PhysicsAsset /Game/Characters/Player/Female/Medium/Bodies/F_MED_Little_Egg/Meshes/F_MED_Little_Egg_Physics
bodies {
 Dyn_Bracelet_Main [{shapeType=Capsule, length=2, radius=2, shapeName=Dyn_Bracelet_Main_capsule, center=<0.5 0 0>, rotation=<rot 90 0 0>}]
 Dyn_MasterBracelet_Main [{shapeType=Capsule, length=0.5, radius=0.5, shapeName=Dyn_MasterBracelet_Main_capsule}]
 Dyn_anklet1 [{shapeType=Capsule, length=1, radius=1, shapeName=Dyn_anklet1_capsule, center=<0.5 0 0>, rotation=<rot 90 0 0>}]
 Dyn_belt1 [{shapeType=Capsule, length=1, radius=1, shapeName=Dyn_belt1_capsule, center=<0.5 0 0>, rotation=<rot 90 0 0>}]
 Dyn_master_anklet [{shapeType=Capsule, length=0.5, radius=0.5, shapeName=Dyn_master_anklet_capsule}]
 Dyn_master_belt [{shapeType=Capsule, length=0.5, radius=0.5, shapeName=Dyn_master_belt_capsule}]
 dyn_bracelette_r1 [{shapeType=Capsule, length=1, radius=1, shapeName=dyn_bracelette_r1_capsule, center=<0.5 0 0>, rotation=<rot 90 0 0>}]
 dyn_cuff_l1 [{shapeType=Capsule, length=1, radius=1, shapeName=dyn_cuff_l1_capsule, center=<0.5 0 0>, rotation=<rot 90 0 0>}]
 dyn_cuff_r1 [{shapeType=Capsule, length=1, radius=1, shapeName=dyn_cuff_r1_capsule, center=<0.5 0 0>, rotation=<rot 90 0 0>}]
 dyn_master_bracelette_r [{shapeType=Capsule, length=1, radius=1, shapeName=dyn_master_bracelette_r_capsule, center=<0.5 0 0>, rotation=<rot 90 0 0>}]
 dyn_master_cuff_l [{shapeType=Capsule, length=0.5, radius=0.5, shapeName=dyn_master_cuff_l_capsule}]
 dyn_master_cuff_r [{shapeType=Capsule, length=0.5, radius=0.5, shapeName=dyn_master_cuff_r_capsule}]
 dyn_master_neclace [{shapeType=Capsule, length=0.5, radius=0.5, shapeName=dyn_master_neclace_capsule}]
 dyn_master_zipper [{shapeType=Capsule, length=0.5, radius=0.5, shapeName=dyn_master_zipper_capsule}]
 dyn_neclace_01 [{shapeType=Capsule, length=1, radius=1, shapeName=dyn_neclace_01_capsule, center=<0.5 0 0>, rotation=<rot 90 0 0>}]
 dyn_zipper_1 [{shapeType=Capsule, length=1, radius=1, shapeName=dyn_zipper_1_capsule, center=<0.5 0 0>, rotation=<rot 90 0 0>}]
}
constraints [
 {bone1Name=dyn_cuff_r1, bone2Name=dyn_master_cuff_r, swing1LimitDegrees=20, swing2LimitDegrees=1, twistLimitDegrees=1}
 {bone1Name=dyn_bracelette_r1, bone2Name=dyn_master_bracelette_r, swing1LimitDegrees=45, swing2LimitDegrees=45, twistLimitDegrees=1}
 {bone1Name=dyn_neclace_01, bone2Name=dyn_master_neclace, swing1LimitDegrees=20, swing2LimitDegrees=1, twistLimitDegrees=1}
 {bone1Name=dyn_zipper_1, bone2Name=dyn_master_zipper, swing1LimitDegrees=20, swing2LimitDegrees=1, twistLimitDegrees=1}
 {bone1Name=Dyn_belt1, bone2Name=Dyn_master_belt, swing1LimitDegrees=20, swing2LimitDegrees=1, twistLimitDegrees=1}
 {bone1Name=Dyn_anklet1, bone2Name=Dyn_master_anklet, swing1LimitDegrees=20, swing2LimitDegrees=1, twistLimitDegrees=1}
 {bone1Name=dyn_cuff_l1, bone2Name=dyn_master_cuff_l, swing1LimitDegrees=20, swing2LimitDegrees=1, twistLimitDegrees=1}
 {
  bone1Name=dyn_master_bracelette_r
  bone2Name=Dyn_Bracelet_Main
  swing1LimitDegrees=45
  swing1Motion=Free
  swing2LimitDegrees=45
  swing2Motion=Free
  twistLimitDegrees=45
  twistMotion=Free
 }
 {
  bone1Name=Dyn_Bracelet_Main
  bone2Name=Dyn_MasterBracelet_Main
  swing1LimitDegrees=5
  swing1Motion=Locked
  swing2LimitDegrees=5
  swing2Motion=Locked
  twistLimitDegrees=5
  twistMotion=Locked
 }
]
