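# Post-body animation Blueprint for the Little Egg female cosmetic body, adding secondary motion to its bracelets, cuffs, necklace, belt and anklet.
# It copies the attached parent mesh's pose, places accessory master bones with weighted multi-bone constraints, and caches the result.
# A chain of Anim Dynamics nodes simulates each dangling bone from that cache into the output.
# A RigidBody branch with a front-end bool blend exists but does not reach the output.
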AnimBlueprint /BRCosmetics/Characters/Player/Female/Medium/Bodies/F_MED_Little_Egg/Meshes/F_MED_Little_Egg_AnimBP
target_skeleton=&/F_MED_Little_Egg_Skeleton
graphs [
 AnimGraph {nodes [
   n1 = RigidBody(ComponentPose=Local To Component(LocalPose=Use cached pose 'PreSim Cache'())) {
    bClampLinearTranslationLimitToRefPose=true
    bFreezeIncomingPoseOnStart=true
    bOverrideWorldGravity=true
    overridePhysicsAsset=/Game/Characters/Player/Female/Medium/Bodies/F_MED_Little_Egg/Meshes/F_MED_Little_Egg_Physics
    simulationSpace=BaseBoneSpace
    alphaBoolBlend {blendOption=Linear}
    baseBoneRef {boneName=pelvis}
    componentAppliedLinearAccClamp=<1000 1000 1000>
    componentLinearAccScale=<0.1 0.1 0.1>
    componentLinearVelScale=<0.1 0.1 0.1>
    overrideWorldGravity {z=-980}
    simSpaceSettings {maxAngularAcceleration=1000, maxAngularVelocity=1000, maxLinearAcceleration=1000, maxLinearVelocity=1000}
   }
   Blend Poses by bool(BlendPose_0=Component To Local(ComponentPose=@n1), bActiveValue=Get bIsInFrontEnd()) {blendType=Linear}
   n2 = Anim Dynamics - Bone: Dyn_Bracelet_Main(ComponentPose=Local To Component(LocalPose=Use cached pose 'PreSim Cache'()), Alpha=1, ExternalForce="0,0,0") {
    angularDampingOverride=0.998
    bOverrideAngularDamping=true
    bUseGravityOverride=true
    linearDampingOverride=0.995
    linearSpringConstant=1
    simulationSpace=RootRelative
    alphaBoolBlend {blendOption=Linear}
    boundBone {boneName=Dyn_Bracelet_Main}
    componentAppliedLinearAccClamp=<1000 1000 1000>
    componentLinearAccScale=<0.1 0.1 0.1>
    componentLinearVelScale=<0.1 0.1 0.1>
    gravityOverride {z=-800}
    physicsBodyDefinitions [{
      boundBone {boneName=Dyn_Bracelet_Main}
      boxExtents=<15 40 40>
      constraintSetup {angularLimitsMax=<2 5 15>, angularLimitsMin=<-2 -5 -10>, angularTarget=<0 0 0>}
      localJointOffset=<0 5 0>
     }]
   }
   n3 = Anim Dynamics - Bone: dyn_bracelette_r1(ComponentPose=@n2, Alpha=1, ExternalForce="0,0,0") {
    angularDampingOverride=0.99
    bOverrideAngularDamping=true
    bUseGravityOverride=true
    simulationSpace=RootRelative
    alphaBoolBlend {blendOption=Linear}
    boundBone {boneName=dyn_bracelette_r1}
    componentAppliedLinearAccClamp=<1000 1000 1000>
    componentLinearAccScale=<0.1 0.1 0.1>
    componentLinearVelScale=<0.1 0.1 0.1>
    gravityOverride {z=-980}
    physicsBodyDefinitions [{
      boundBone {boneName=dyn_bracelette_r1}
      boxExtents=<5 15 15>
      constraintSetup {angularLimitsMax=<2 45 25>, angularLimitsMin=<-2 -45 -45>, angularTarget=<0 0 0>}
      localJointOffset=<5 0 0>
     }]
   }
   n4 = Anim Dynamics - Bone: dyn_cuff_r1(ComponentPose=@n3, Alpha=1, ExternalForce="0,0,0") {
    angularDampingOverride=0.994
    bOverrideAngularDamping=true
    bUseGravityOverride=true
    alphaBoolBlend {blendOption=Linear}
    boundBone {boneName=dyn_cuff_r1}
    componentAppliedLinearAccClamp=<1000 1000 1000>
    componentLinearAccScale=<0.1 0.1 0.1>
    componentLinearVelScale=<0.1 0.1 0.1>
    gravityOverride {z=-980}
    physicsBodyDefinitions [{
      boundBone {boneName=dyn_cuff_r1}
      boxExtents=<10 20 20>
      constraintSetup {angularLimitsMax=<2 2 15>, angularLimitsMin=<-2 -2 -10>, angularTarget=<0 0 0>}
      localJointOffset=<5 0 0>
     }]
   }
   n5 = Anim Dynamics - Bone: dyn_cuff_l1(ComponentPose=@n4, Alpha=1, ExternalForce="0,0,0") {
    angularDampingOverride=0.994
    bOverrideAngularDamping=true
    bUseGravityOverride=true
    alphaBoolBlend {blendOption=Linear}
    boundBone {boneName=dyn_cuff_l1}
    componentAppliedLinearAccClamp=<1000 1000 1000>
    componentLinearAccScale=<0.1 0.1 0.1>
    componentLinearVelScale=<0.1 0.1 0.1>
    gravityOverride {z=-980}
    physicsBodyDefinitions [{
      boundBone {boneName=dyn_cuff_l1}
      boxExtents=<10 20 20>
      constraintSetup {angularLimitsMax=<2 2 15>, angularLimitsMin=<-2 -2 -10>, angularTarget=<0 0 0>}
      localJointOffset=<5 0 0>
     }]
   }
   n6 = Anim Dynamics - Bone: dyn_neclace_01!(ComponentPose=@n5, Alpha=1, GravityOverride:Vector="0,0,-980", ExternalForce="0,0,0") -> (Pose) {
    bOverrideLinearDamping=true
    bUseGravityOverride=true
    bUsePlanarLimit=false
    linearDampingOverride=0.9999
    simulationSpace=RootRelative
    alphaBoolBlend {blendOption=Linear}
    boundBone {boneName=dyn_neclace_01}
    componentAppliedLinearAccClamp=<3000 3000 3000>
    componentLinearAccScale=<0.2 0.2 0.25>
    componentLinearVelScale=<0.2 0.2 0.25>
    gravityOverride {z=-980}
    physicsBodyDefinitions [{
      boundBone {boneName=dyn_neclace_01}
      boxExtents=<25 10 25>
      constraintSetup {angularLimitsMax=<1 10 1>, angularLimitsMin=<-1 -10 -30>, angularTarget=<0 0 0>}
      localJointOffset=<4 0 0>
     }]
    sphericalLimits [{limitRadius=5, drivingBone {boneName=hand_l}, sphereLocalOffset=<0 0 0>}]
   }
   n7 = Anim Dynamics - Bone: Dyn_belt1!(ComponentPose=@n6, Alpha=1, GravityOverride:Vector="0,0,-980", ExternalForce="0,0,0") -> (Pose) {
    bOverrideLinearDamping=true
    bUseGravityOverride=true
    bUsePlanarLimit=false
    linearDampingOverride=0.9999
    simulationSpace=RootRelative
    alphaBoolBlend {blendOption=Linear}
    boundBone {boneName=Dyn_belt1}
    componentAppliedLinearAccClamp=<3000 3000 3000>
    componentLinearAccScale=<0.2 0.2 0.25>
    componentLinearVelScale=<0.2 0.2 0.25>
    gravityOverride {z=-980}
    physicsBodyDefinitions [{
      boundBone {boneName=Dyn_belt1}
      boxExtents=<30 10 30>
      constraintSetup {angularLimitsMax=<1 5 10>, angularLimitsMin=<-1 -5 -10>, angularTarget=<0 0 0>}
      localJointOffset=<4 0 0>
     }]
    sphericalLimits [{limitRadius=5, drivingBone {boneName=hand_l}, sphereLocalOffset=<0 0 0>}]
   }
   n8 = Anim Dynamics - Bone: Dyn_anklet1!(ComponentPose=@n7, Alpha=1, GravityOverride:Vector="0,0,-800", ExternalForce="0,0,0") -> (Pose) {
    bOverrideLinearDamping=true
    bUseGravityOverride=true
    bUsePlanarLimit=false
    linearDampingOverride=0.9999
    simulationSpace=RootRelative
    alphaBoolBlend {blendOption=Linear}
    boundBone {boneName=Dyn_anklet1}
    componentAppliedLinearAccClamp=<1500 1500 1500>
    componentLinearAccScale=<0.1 0.1 0.1>
    componentLinearVelScale=<0.1 0.1 0.1>
    gravityOverride {z=-980}
    physicsBodyDefinitions [{
      boundBone {boneName=Dyn_anklet1}
      boxExtents=<15 5 15>
      constraintSetup {angularLimitsMax=<1 15 0>, angularLimitsMin=<-1 -15 -25>, angularTarget=<0 0 0>}
      localJointOffset=<4 0 0>
     }]
    sphericalLimits [{limitRadius=5, drivingBone {boneName=hand_l}, sphereLocalOffset=<0 0 0>}]
   }
   Output Pose(Result=Component To Local(ComponentPose=@n8)) {layerGroup=None}
   n9 = "Constraint - Dyn_MasterBracelet_Main (PPP)"!(
    ComponentPose=Local To Component(LocalPose=Copy Pose From Mesh() {bUseAttachedParent=true})
    Alpha=1
    ConstraintWeights_0=0.7
    ConstraintWeights_1:float=0.3
    ConstraintWeights_2:float=0
   ) -> (Pose) {
    alphaBoolBlend {blendOption=Linear}
    boneToModify {boneName=Dyn_MasterBracelet_Main}
    constraintSetup [{targetBone {boneName=lowerarm_twist_01_r}}, {targetBone {boneName=hand_r}}, {targetBone {boneName=thumb_01_r}}]
    constraintWeights [1,1,1]
   }
   n10 = "Constraint - dyn_master_neclace (PPP)"!(ComponentPose=@n9, Alpha=1, ConstraintWeights_0=0.934, ConstraintWeights_1:float=0.0313, ConstraintWeights_2:float=0.0347) -> (Pose) {
    alphaBoolBlend {blendOption=Linear}
    boneToModify {boneName=dyn_master_neclace}
    constraintSetup [{targetBone {boneName=spine_05}}, {targetBone {boneName=clavicle_l}}, {targetBone {boneName=clavicle_r}}]
    constraintWeights [1,1,1]
   }
   n11 = "Constraint - Dyn_master_belt (PP)"!(ComponentPose=@n10, Alpha=1, ConstraintWeights_0=0.0153, ConstraintWeights_1:float=0.9847) -> (Pose) {
    alphaBoolBlend {blendOption=Linear}
    boneToModify {boneName=Dyn_master_belt}
    constraintSetup [{targetBone {boneName=spine_02}}, {targetBone {boneName=pelvis}}]
    constraintWeights [1,1]
   }
   n12 = "Constraint - Dyn_master_anklet (PPPP)"!(
    ComponentPose=@n11
    Alpha=1
    ConstraintWeights_0=0.0271
    ConstraintWeights_1:float=0.6465
    ConstraintWeights_2:float=0.2647
    ConstraintWeights_3:float=0.0617
   ) -> (Pose) {
    alphaBoolBlend {blendOption=Linear}
    boneToModify {boneName=Dyn_master_anklet}
    constraintSetup [
     {targetBone {boneName=calf_r}}
     {targetBone {boneName=calf_twist_01_r}}
     {targetBone {boneName=foot_r}}
     {targetBone {boneName=calf_twist_02_r}}
    ]
    constraintWeights [1,1,1,1]
   }
   Save cached pose 'PreSim Cache'(Pose=Component To Local(ComponentPose=@n12)) {cacheName=PreSim Cache}
  ]}
 EventGraph {nodes [Event BlueprintUpdateAnimation()]}
]
variables []
